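# Niagara system for the Grass vehicle boost cosmetic, throwing grass, dust and sand sprites from the car's boost sockets.
# One emitter runs five spawn groups, written to VisibilityTag alongside five sprite renderers with SubUV animation.
# Particles are partitioned across sockets, launched along negative speed, then shaped by gravity, curl noise and inherited movement.
# Per-layer colors switch between default and painted variants, with opacity curves and game-mode-indexed values.
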NiagaraSystem /VehicleCosmetics/Boosts/Boost_Grass/FX/NS_Boost_Grass_v02
dependencies {
 usedDataInterfaces [&NiagaraDataInterfaceSpriteRendererInfo, &NiagaraDataInterfaceSkeletalMesh]
 usedDynamicInputs [&MergeHelper_VectorSwitch]
 usedModules [
  &self:SystemSpawnScript.SystemScriptSource.SystemScriptGraph.NiagaraNodeAssignment_0.SetVariables_#1
  &SystemState
 ]
}
emitters {Grass001 {
  emitterUpdateScript [
   &EmitterState(Life Cycle Mode=::NewEnumerator1, UseLoopDelay=false)
   &VC_FilteredSocketCount()
   &SpawnRate(SpawnRate=45, Use Spawn Probability=false)
   &SpawnRate(SpawnGroup=1, SpawnRate=45, Use Spawn Probability=false) {module=SpawnRate001}
   &SpawnRate(SpawnGroup=2, SpawnRate=45, Use Spawn Probability=false) {module=SpawnRate002}
   &SpawnRate(SpawnGroup=3, SpawnRate=45, Use Spawn Probability=false) {module=SpawnRate003}
   &SpawnRate(SpawnGroup=4, SpawnRate=70, Use Spawn Probability=false) {module=SpawnRate004}
   /SetVariables(
    "Emitter.Durst (3)":LinearColor=&MergeHelper_ColorSwitch(0=<rgba 1 0 0 0>, 1=<rgba 3 0 0 0>, 2=<rgba 5 0 0 0>, 3=?, Index=$User.GameModeID, Num=::NewEnumerator1)
    "Emitter.Grass (0)":LinearColor=&MergeHelper_ColorSwitch(0=<rgba 2 0 0 0>, 1=<rgba 5 0 0 0>, 2=<rgba 5 0 0 0>, 3=?, Index=$User.GameModeID, Num=::NewEnumerator1)
    "Emitter.Grass (1)":LinearColor=&MergeHelper_ColorSwitch(0=<rgba 2 0 0 0>, 1=<rgba 10 0 0 0>, 2=<rgba 5 0 0 0>, 3=?, Index=$User.GameModeID, Num=::NewEnumerator1)
    "Emitter.Grass (2)":LinearColor=&MergeHelper_ColorSwitch(0=<rgba 2 0 0 0>, 1=<rgba 10 0 0 0>, 2=<rgba 5 0 0 0>, 3=?, Index=$User.GameModeID, Num=::NewEnumerator1)
    "Emitter.Sand (4)":LinearColor=&MergeHelper_ColorSwitch(0=<rgba 1 0 0 0>, 1=<rgba 3 0 0 0>, 2=<rgba 5 0 0 0>, 3=?, Index=$User.GameModeID, Num=::NewEnumerator1)
   ) {isSetParameters=true}
  ]
  particleSpawnScript [
   &VC_PaintedVariantsOverride(
    BurntSienna=<rgba 0.5677 0.3183 0.2513 1>
    Cobalt=<rgba 0.3767 0.3883 1 1>
    Crimson=<rgba 0.599 0.1508 0.1629 1>
    Grey=<rgba 0.401 0.401 0.401 1>
    Lime=<rgba 0.5899 0.7552 0.1036 1>
    Orange=<rgba 1 0.7468 0 1>
    Pink=<rgba 1 0.5122 0.8226 1>
    Purple=<rgba 0.4104 0 1 1>
    R1 BurntSienna:LinearColor=<rgba 0.592 0.3449 0.2785 1>
    R1 Cobalt:LinearColor=<rgba 0.3281 0.3406 1 1>
    R1 Crimson:LinearColor=<rgba 0.5955 0.1602 0.172 1>
    R1 ForestGreen:LinearColor=<rgba 0 0.5 0.0253 1>
    R1 Grey:LinearColor=<rgba 0.4497 0.4497 0.4497 1>
    R1 Lime:LinearColor=<rgba 0.7373 0.9045 0 1>
    R1 Orange:LinearColor=<rgba 1 0.5392 0 1>
    R1 Pink:LinearColor=<rgba 1 0.6128 0.8592 1>
    R1 Purple:LinearColor=<rgba 0.4104 0 1 1>
    R1 Saffron:LinearColor=<rgba 0.8632 1 0 1>
    R1 SkyBlue:LinearColor=<rgba 0 0.6654 1 1>
    R2 BurntSienna:LinearColor=<rgba 0.2448 0.0984 0.0591 1>
    R2 Cobalt:LinearColor=<rgba 0.2205 0.2349 1 1>
    R2 Crimson:LinearColor=<rgba 0.4635 0.136 0.1449 1>
    R2 Lime:LinearColor=<rgba 0.3531 0.4566 0 1>
    R2 Orange:LinearColor=<rgba 0.533 0.3482 0.2026 1>
    R2 Pink:LinearColor=<rgba 0.5747 0.2424 0.4538 1>
    R2 Purple:LinearColor=<rgba 0.2743 0 0.6684 1>
    R2 Saffron:LinearColor=<rgba 0.469 0.5434 0 1>
    R2 SkyBlue:LinearColor=<rgba 0 0.4979 0.7483 1>
    R2 TitaniumWhite:LinearColor=<rgba 0.3698 0.3698 0.3698 1>
    R3 Black:LinearColor=<rgba 0.25 0.25 0.25 1>
    R3 BurntSienna:LinearColor=<rgba 0.717 0.4746 0.4095 1>
    R3 Cobalt:LinearColor=<rgba 0.3559 0.3678 1 1>
    R3 Crimson:LinearColor=<rgba 0.533 0.1064 0.118 1>
    R3 ForestGreen:LinearColor=<rgba 0 0.533 0.0269 1>
    R3 Grey:LinearColor=<rgba 0.4184 0.4184 0.4184 1>
    R3 Lime:LinearColor=<rgba 0.303 0.6059 0 1>
    R3 Orange:LinearColor=<rgba 0.6719 0.5116 0.4258 1>
    R3 Pink:LinearColor=<rgba 1 0.5677 0.8428 0.5304>
    R3 Purple:LinearColor=<rgba 0.563 0.4171 0.7726 0.7739>
    R3 Saffron:LinearColor=<rgba 1 0.887 0.5156 0.4435>
    R3 SkyBlue:LinearColor=<rgba 0.4288 0.8089 1 0.4667>
    R3 TitaniumWhite:LinearColor=<rgba 0.75 0.75 0.75 0.4667>
    Saffron=<rgba 0.8632 1 0 1>
    SkyBlue=<rgba 0 0.6654 1 1>
    Use R1 Painted Overrides=true
    Use R2 Painted Overrides=true
    Use R3 Painted Overrides=true
   )
   /SetVariables(Particles.VisibilityTag=$Emitter.SpawnGroup) {isSetParameters=true}
   &InitializeParticle(
    Color=$Particles.Color
    Lifetime=&MergeHelper_FloatSwitch(
     0=&RandomRangeFloat(Minimum=0.25, Maximum=1)
     1=&RandomRangeFloat(Minimum=0.25, Maximum=1)
     2=&RandomRangeFloat(Minimum=0.25, Maximum=1)
     3=&RandomRangeFloat(Minimum=0.2, Maximum=0.3)
     4=&RandomRangeFloat(Minimum=0.5, Maximum=1.2)
     Index=$Emitter.SpawnGroup
     Num=::NewEnumerator3
    )
    Sprite Rotation Mode=::NewEnumerator2
    Sprite Size Mode=::NewEnumerator3
    Uniform Sprite Size=&MergeHelper_FloatSwitch(
     0=80
     1=100
     2=80
     3=&RandomRangeFloat(Minimum=20, Maximum=10)
     4=120
     Index=$Emitter.SpawnGroup
     Num=::NewEnumerator3
    )
    UsePositionOffset=false
    Write Material Random=true
   )
   &PartitionParticles(Number Of Partitions=$Emitter.PortCount)
   &SkeletalMeshLocation(
    Bone / Socket Index=$Particles.PartitionParticles.Partition
    Bone / Socket Sampling Mode=::NewEnumerator5
    MaskSpawnGroup=false
    Mesh Sampling Type=::NewEnumerator6
    Sampled Position Offset=&AddVectorToPosition(Position=<-60 0 0>, Vector=$User.DTCarOffset)
    Skeletal Mesh=$System.SkeletalMesh
   )
   &AddVelocity(
    Cone Angle=&MergeHelper_FloatSwitch(0=0, 1=0, 2=0, 3=15, Index=$Emitter.SpawnGroup, Num=::NewEnumerator2)
    Distribution Along Cone Axis=200
    Module.UseConeFalloff=false
    Module.UseVelDistribution=true
    Override Seed=false
    Velocity Mode=::NewEnumerator2
    Velocity Speed=-2000
   )
   &AddVelocity(Velocity=&RandomRangeVector(Minimum=<0 0 5>, Maximum=<0 0 25>), Velocity Speed Scale=10) {module=AddVelocity001}
   &ShapeLocation(
    MaskSpawnGroup=false
    Override Seed=false
    Rotation Mode=::NewEnumerator1
    Sphere Radius=&MergeHelper_FloatSwitch(0=0, 1=0, 2=0, 3=40, Index=$Emitter.SpawnGroup, Num=::NewEnumerator2)
    UseOwnerScale=true
   )
   /SetVariables() {isSetParameters=true}
  ]
  particleUpdateScript [
   &ParticleState()
   &AccelerationForce(Acceleration=&RandomRangeVector(Minimum=<0 0 -512>, Maximum=<0 0 -1024>))
   &InheritSourceMovement(
    Applied Position Scale=&Multiply_VectorByFloat(
     Vector=<1 1 1>
     Float=&FloatFromCurve(FloatCurve=DataInterface() {propertyValues {Curve {keys [
          {arriveTangent=0.0056, interpMode=RCIM_Cubic, leaveTangent=0.0056, tangentMode=RCTM_User, time=0.1, value=1}
          {arriveTangent=-1.8464, interpMode=RCIM_Cubic, leaveTangent=-1.8464, tangentMode=RCTM_User, time=1, value=0.5}
         ]}}})
    )
    ApplyForce=false
    ApplyPosition=true
    ApplyPositionFalloff=false
   )
   &CurlNoiseForce(Bias Noise=false, Noise Frequency=100, Noise Strength=100, Pan Noise=false)
   &SolveForcesAndVelocity(Clamp Velocity=false, Limit Acceleration=false)
   /SetVariables(
    "Particles.Durst (3)":LinearColor=&MakeCustomLinearColorFromBool(
     Bool=$User.bIsPainted
     True Color=&MultiplyLinearColor(
      Linear Color=$Particles.R2 Color
      Scale Factor=&MakeVector4FromVectorAndFloat("Vector (RGB)"=<1 1 1>, "Float (Alpha)"=&FloatFromCurve(FloatCurve=$User.General Opacity Color))
     )
     False Color=&MultiplyLinearColor(
      Linear Color=<rgba 0.3351 0.3351 0.3351 1>
      Scale Factor=&MakeVector4FromVectorAndFloat("Vector (RGB)"=<1 1 1>, "Float (Alpha)"=&FloatFromCurve(FloatCurve=$User.General Opacity Color))
     )
    )
    "Particles.Grass (0)":LinearColor=&MakeCustomLinearColorFromBool(
     Bool=$User.bIsPainted
     True Color=&MultiplyLinearColor(
      Linear Color=$Particles.Color
      Scale Factor=&MakeVector4FromVectorAndFloat("Vector (RGB)"=<1 1 1>, "Float (Alpha)"=&FloatFromCurve(FloatCurve=$User.General Opacity Color))
     )
     False Color=&MultiplyLinearColor(
      Linear Color=<rgba 0.1414 0.6962 0.1688 1>
      Scale Factor=&MakeVector4FromVectorAndFloat("Vector (RGB)"=<1 1 1>, "Float (Alpha)"=&FloatFromCurve(FloatCurve=$User.General Opacity Color))
     )
    )
    "Particles.Grass (1)":LinearColor=&MakeCustomLinearColorFromBool(
     Bool=$User.bIsPainted
     True Color=&MultiplyLinearColor(
      Linear Color=$Particles.R1 Color
      Scale Factor=&MakeVector4FromVectorAndFloat("Vector (RGB)"=<1 1 1>, "Float (Alpha)"=&FloatFromCurve(FloatCurve=$User.General Opacity Color))
     )
     False Color=&MultiplyLinearColor(
      Linear Color=<rgba 0.2039 0.5781 0.1355 1>
      Scale Factor=&MakeVector4FromVectorAndFloat("Vector (RGB)"=<1 1 1>, "Float (Alpha)"=&FloatFromCurve(FloatCurve=$User.General Opacity Color))
     )
    )
    "Particles.Grass (2)":LinearColor=&MakeCustomLinearColorFromBool(
     Bool=$User.bIsPainted
     True Color=&MultiplyLinearColor(
      Linear Color=$Particles.R2 Color
      Scale Factor=&MakeVector4FromVectorAndFloat("Vector (RGB)"=<1 1 1>, "Float (Alpha)"=&FloatFromCurve(FloatCurve=$User.General Opacity Color))
     )
     False Color=&MultiplyLinearColor(
      Linear Color=<rgba 0.5573 0.5529 0.0281 1>
      Scale Factor=&MakeVector4FromVectorAndFloat("Vector (RGB)"=<1 1 1>, "Float (Alpha)"=&FloatFromCurve(FloatCurve=$User.General Opacity Color))
     )
    )
    "Particles.Sand (4)":LinearColor=&MakeCustomLinearColorFromBool(
     Bool=$User.bIsPainted
     True Color=&MultiplyLinearColor(
      Linear Color=$Particles.R3 Color
      Scale Factor=&MakeVector4FromVectorAndFloat(
       "Vector (RGB)"=<1 1 1>
       "Float (Alpha)"=&FloatFromCurve(FloatCurve=$User.General Opacity Color, Scale Curve=0.1)
      )
     )
     False Color=&MultiplyLinearColor(
      Linear Color=<rgba 0.5 0.422 0.25 1>
      Scale Factor=&MakeVector4FromVectorAndFloat(
       "Vector (RGB)"=<1 1 1>
       "Float (Alpha)"=&FloatFromCurve(FloatCurve=$User.General Opacity Color, Scale Curve=0.1)
      )
     )
    )
   ) {isSetParameters=true}
   &SubUVAnimation(
    "Sprite Renderer "=DataInterface() {propertyValues {SpriteRenderer=&self:Grass001_0.NiagaraSpriteRendererProperties_1}}
    SubUV Animation Mode:ENiagara_SubUVLookupModeV2=::NewEnumerator2
    UseEndFrame=false
    UseStartFrame=false
   )
   &SubUVAnimation(
    "Sprite Renderer "=DataInterface() {propertyValues {SpriteRenderer=&self:Grass001_0.NiagaraSpriteRendererProperties_0}}
    SubUV Animation Mode:ENiagara_SubUVLookupModeV2=::NewEnumerator2
    UseEndFrame=false
    UseStartFrame=false
   ) {module=SubUVAnimation001}
   &SubUVAnimation(
    "Sprite Renderer "=DataInterface() {propertyValues {SpriteRenderer=&self:Grass001_0.NiagaraSpriteRendererProperties_2}}
    SubUV Animation Mode:ENiagara_SubUVLookupModeV2=::NewEnumerator2
    UseEndFrame=false
    UseStartFrame=false
   ) {module=SubUVAnimation002}
   &SubUVAnimation(
    "Sprite Renderer "=DataInterface() {propertyValues {SpriteRenderer=&self:Grass001_0.NiagaraSpriteRendererProperties_3}}
    SubUV Animation Mode:ENiagara_SubUVLookupModeV2=::NewEnumerator2
    UseEndFrame=false
    UseStartFrame=false
   ) {module=SubUVAnimation003}
   &SubUVAnimation(
    "Sprite Renderer "=DataInterface() {propertyValues {SpriteRenderer=&self:Grass001_0.NiagaraSpriteRendererProperties_4}}
    UseEndFrame=false
    UseStartFrame=false
   ) {module=SubUVAnimation004}
  ]
  rendererClasses [&NiagaraSpriteRendererProperties]
  renderers [
   {rendererClass=&NiagaraSpriteRendererProperties}
   {rendererClass=&NiagaraSpriteRendererProperties, rendererIndex=1}
   {rendererClass=&NiagaraSpriteRendererProperties, rendererIndex=2}
   {rendererClass=&NiagaraSpriteRendererProperties, rendererIndex=3}
   {rendererClass=&NiagaraSpriteRendererProperties, rendererIndex=4}
  ]
 }}
summary {
 systemName=NS_Boost_Grass_v02
 emitters [{emitterName=Grass001, rendererClasses [&NiagaraSpriteRendererProperties]}]
 userVariables [
  User.bIsFrontendPreview {type=NiagaraBool}
  User.bIsPainted {type=NiagaraBool}
  User.ColorVariantID {type=NiagaraInt32}
  User.DTCarOffset {type=Vector3f, description="", defaultValue {struct=Vector3f, value=<0 0 0>}}
  User.GameModeID {type=NiagaraInt32}
  User.General Opacity Color {
   type=NiagaraDataInterfaceCurve
   description=""
   defaultValue {
    struct=NiagaraExt_VariableValue_DataInterface
    value {
     dataInterface=&self:NiagaraDataInterfaceCurve_0
     dataInterfaceClass=&NiagaraDataInterfaceCurve
    }
   }
  }
  User.PaintedColor {type=LinearColor, description="", defaultValue {struct=LinearColor, value=<rgba 1 1 1 1>}}
  User.PaintedScalar {type=NiagaraFloat, defaultValue=1, description=""}
  User.PortCount {type=NiagaraInt32, defaultValue=2}
 ]
}
